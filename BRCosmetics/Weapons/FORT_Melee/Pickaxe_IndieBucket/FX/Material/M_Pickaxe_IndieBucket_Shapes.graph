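# Masked, unlit material for Niagara mesh particles in the IndieBucket pickaxe effects, showing shapes from a flipbook atlas.
# The Shape dynamic parameter is remapped to 0–1 between Input Low and Input High, then drives the FlipBook phase over the Rows and Columns grid, letting particle data pick the variant.
# The flipbook result is halved and fed into MF_FX_EmissiveAdjustments; the ShapeTexture scalar is unconnected.
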Material /BRCosmetics/Weapons/FORT_Melee/Pickaxe_IndieBucket/FX/Material/M_Pickaxe_IndieBucket_Shapes
graph {
 properties {
  bUsedWithNiagaraMeshParticles=true
  blendMode=BLEND_Masked
  pixelDepthOffsetMode=PDOM_Legacy
  shadingModel=MSM_Unlit
  translucencyLightingMode=TLM_VolumetricDirectional
  usedShadingModels=Unlit
 }
 nodes [
  n1 = MaterialFunctionCall(
   Input=DynamicParameter() {paramNames [Shape,Param2,Param3,Param4]}.Shape
   Input Low=ScalarParameter() {parameterName=Input Low}
   Input High=ScalarParameter() {defaultValue=4, parameterName=Input High}
  ) {materialFunction=&RemapValue_0-1}
  n2 = ScalarParameter() {defaultValue=2, parameterName=Rows}
  n3 = ScalarParameter() {defaultValue=2, parameterName=Columns}
  n4 = TextureObjectParameter() {parameterName=VarientTexture, texture=/BRCosmetics/Weapons/FORT_Melee/Pickaxe_IndieBucket/FX/Texture/T_Pickaxe_IndieBucket_Shapes}
  MaterialFunctionCall(Emissive=Multiply(A=MaterialFunctionCall("Animation  Phase (0-1)"=@n1.Result, Number of Rows=@n2, Number of Columns=@n3, Texture=@n4) {materialFunction=&FlipBook}.Result) {constB=0.5}) {materialFunction=&MF_FX_EmissiveAdjustments}
  ScalarParameter() {defaultValue=1, parameterName=ShapeTexture}
 ]
}
