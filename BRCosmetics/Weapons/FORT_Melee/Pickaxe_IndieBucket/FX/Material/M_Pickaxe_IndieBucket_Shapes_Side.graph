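# Unlit masked material for the side faces of shape mesh particles in the IndieBucket pickaxe effects, drawing colored edge bands around a base color.
# The V coordinate runs through a bias-scale and sine, then an inverted absolute value, so a smoothstep with adjustable amount and softness marks both V edges.
# That mask blends Color with ColorEdge, and the sum feeds the MF_FX_EmissiveAdjustments function.
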Material /BRCosmetics/Weapons/FORT_Melee/Pickaxe_IndieBucket/FX/Material/M_Pickaxe_IndieBucket_Shapes_Side
graph {
 properties {
  bUsedWithNiagaraMeshParticles=true
  blendMode=BLEND_Masked
  pixelDepthOffsetMode=PDOM_Legacy
  shadingModel=MSM_Unlit
  usedShadingModels=Unlit
 }
 nodes [
  n1 = VectorParameter() {parameterName=Color, defaultValue=<rgba 0.8793 0.8795 0.9333 1>}
  n2 = ScalarParameter() {defaultValue=0.78, parameterName=Edge Amount}
  n3 = TextureCoordinate()
  n4 = Sine(None=ConstantBiasScale(None=MaterialFunctionCall(Float2=@n3) {materialFunction=&BreakOutFloat2Components}.G))
  n5 = SmoothStep(Min=@n2, Max=Add(A=@n2, B=ScalarParameter() {defaultValue=0.2, parameterName=Edge Softness}), Value=OneMinus(None=Abs(None=@n4)))
  n6 = OneMinus(None=@n5)
  n7 = VectorParameter() {parameterName=ColorEdge, defaultValue=<rgba 1 0 0.0471 1>}
  MaterialFunctionCall(Emissive=Add(A=Multiply(A=@n1.RGB, B=@n6), B=Multiply(A=@n7.RGB, B=@n5))) {materialFunction=&MF_FX_EmissiveAdjustments}
 ]
}
