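# Niagara swing trail for the IndieBucket pickaxe, pairing anim-trail ribbons with mesh particles.
# Round_Trail and Outline follow the skeletal mesh, shaping alpha and width over loop age; Round_Trail adds time-of-day color scaling.
# Pop spawns meshes by distance moved around a sampled skeletal point, with curl noise, drag and scale-out.
# A system variant bool switches outline color, trail alpha and mesh choice; ColorSwitcher blends two Pop color arrays.
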NiagaraSystem /BRCosmetics/Weapons/FORT_Melee/Pickaxe_IndieBucket/FX/NS_Pickaxe_IndieBucket_Trail
dependencies {
 usedDataInterfaces [
  &NiagaraDataInterfaceCurve
  &NiagaraDataInterfaceColorCurve
  &NiagaraDataInterfaceArrayFloat
  &NiagaraDataInterfaceArrayFloat3
 ]
 usedDynamicInputs [
  &SelectFloatFromArray
  &ReturnExecIndex
  &SetBoolByFloatComparison
 ]
 usedModules [
  &self:SystemSpawnScript.SystemScriptSource.SystemScriptGraph.NiagaraNodeAssignment_0.SetVariables_#1
  &self:SystemSpawnScript.SystemScriptSource.SystemScriptGraph.NiagaraNodeAssignment_1.SetVariables_#2
  &SystemState
 ]
}
emitters {
 Outline {
  emitterUpdateScript [
   &EmitterState(Life Cycle Mode=::NewEnumerator1, Loop Behavior=::NewEnumerator1, Loop Duration=1, UseLoopDelay=false)
   &FN_SpawnAnimTrail()
   /SetVariables(Emitter.Random:NiagaraFloat=&RandomRangeFloat(Minimum=0, Maximum=1)) {isSetParameters=true, module=SetVariables_#3}
  ]
  particleSpawnScript [
   &FN_InitializeAnimTrailBoneIndecies()
   &FN_InitializeAnimTrail(Color=<rgba 1 1 1 0.5>, Lifetime=0.12)
   &DynamicMaterialParameters(
    Index 0 Param 4=$Emitter.Random
    Module.Param0WriteEnabled=false
    Module.Param1WriteEnabled=false
    Module.Param2WriteEnabled=false
    Module.Param3WriteEnabled=true
   ) {module=DynamicMaterialParameters001}
  ]
  particleUpdateScript [
   &FN_UpdateAnimTrail(SkeletalMesh=$User.SkeletalMesh) {module=FN_UpdateAnimTrail001}
   /SetVariables(Particles.OutlineColor:LinearColor=&MakeCustomLinearColorFromBool(
     Bool=$System.VarientBase
     True Color=<rgba 0 0 0 0.9>
     False Color=&ColorFromCurve(ColorCurve=DataInterface() {propertyValues {
        AlphaCurve {keys [<at 0 1 RCIM_Cubic>]}
        BlueCurve {keys [<at 0.1051 0.2635 RCIM_Cubic>,<at 0.4991 0>,<at 0.8393 0.4224 RCIM_Cubic>]}
        GreenCurve {keys [<at 0.1051 0.2246 RCIM_Cubic>,<at 0.4991 0.8112>,<at 0.8393 0.1124 RCIM_Cubic>]}
        RedCurve {keys [<at 0.1051 0.031 RCIM_Cubic>,<at 0.4991 0.8802>,<at 0.8393 0.5135 RCIM_Cubic>]}
       }})
    )) {isSetParameters=true}
   &SolveForcesAndVelocity(Clamp Velocity=false, Limit Acceleration=false)
   &ScaleColor(
    Scale Alpha=&Multiply_Float(
     A=&FloatFromCurve(FloatCurve=DataInterface() {propertyValues {Curve {keys [
          {arriveTangent=-1.8862, interpMode=RCIM_Cubic, leaveTangent=-1.8862, tangentMode=RCTM_User, time=0.2, value=1.25}
          {interpMode=RCIM_Cubic, tangentMode=RCTM_User, time=1, value=0}
         ]}}})
     B=&Lerp_Float(
      A=1
      B=&FloatFromCurve(FloatCurve=DataInterface() {propertyValues {Curve {keys [
           {arriveTangent=0.6814, interpMode=RCIM_Cubic, leaveTangent=0.6814, tangentMode=RCTM_User, time=0, value=0.1}
           {arriveTangent=2.5, interpMode=RCIM_Cubic, leaveTangent=2.5, time=0.15, value=0.3}
           {arriveTangent=1.5958, time=0.2, value=0.6}
           {arriveTangent=0.0224, interpMode=RCIM_Cubic, leaveTangent=0.0224, tangentMode=RCTM_User, time=1, value=1}
          ]}}})
      Alpha=&FloatFromCurve(
       FloatCurve=DataInterface() {propertyValues {Curve {keys [<at 0 0 RCIM_Constant>,<at 0.4 1 RCIM_Constant>]}}}
       CurveIndex=$Emitter.NormalizedLoopAge
      )
     )
    )
    Scale RGB=&Multiply_Vector(A=<1 1 1>, B=<1 1 1>)
    ScaleA=true
    ScaleRGB=true
   ) {module=ScaleColor001}
   &ScaleRibbonWidth(Ribbon Width Scale=&FloatFromCurve(
     FloatCurve=DataInterface() {propertyValues {Curve {keys [{arriveTangent=-1.7104, interpMode=RCIM_Cubic, leaveTangent=-1.7104, tangentMode=RCTM_User, time=0.4, value=1}, <at 1 0.5 RCIM_Cubic>]}}}
     CurveIndex=$Emitter.NormalizedLoopAge
    ))
   /SetVariables(Particles.OutlineWidth:NiagaraFloat=&Multiply_Float(A=$Particles.RibbonWidth, B=1)) {isSetParameters=true}
   /SetVariables(Particles.OutlineColor:LinearColor=&MultiplyLinearColor(
     Linear Color=$Particles.OutlineColor
     Scale Factor=&MakeVector4(X=3, Y=3, Z=3, W=&MakeCustomFloatFromBool(Bool=$System.VarientBase, True Float=1, False Float=5))
    )) {enabled=false, isSetParameters=true}
  ]
  rendererClasses [&NiagaraRibbonRendererProperties]
  renderers [{rendererClass=&NiagaraRibbonRendererProperties}]
 }
 Pop {
  emitterUpdateScript [
   &EmitterState()
   &SpawnPerUnit(
    Module.Use Max Movement Threshold=true
    Module.Use Movement Tolerance=false
    Spawn Per Unit=20
    Spawn Probability=0.6
    Use Spawn Probability=true
   )
   /Game/Effects/Niagara/MasterMaterials/ToonShader/Template/NM_ToonShaderParameters(Lighting_Hardness=1, RimLight_Color:LinearColor=<rgba 1 1 1 1>, RimLight_Width=0) {enabled=false}
   /Game/Effects/Niagara/MasterMaterials/ToonShader/Template/NM_OutlineParameters() {enabled=false}
  ]
  particleSpawnScript [
   &FN_SetSpawnGroup(Number of Spawn Groups=6) {enabled=false}
   &PartitionParticles() {enabled=false}
   &InitializeParticle(
    Color=&LerpLinearColors(
     StartColor=&SelectLinearColorFromArray(Array Sampling Mode=::NewEnumerator0, Direct Array Index=?, Color Selection Array=$User.MeshColors, Random Seed=0)
     EndColor=&SelectLinearColorFromArray(Array Sampling Mode=::NewEnumerator0, Direct Array Index=?, Color Selection Array=$User.MeshColors_Variant_01, Random Seed=0)
     LerpFactor=$User.ColorSwitcher
    )
    Lifetime Max=0.5
    Lifetime Min=0.35
    Lifetime Mode=::NewEnumerator1
    Mesh Scale Mode=::NewEnumerator1
    Mesh Uniform Scale Max=0.45
    Mesh Uniform Scale Min=0.25
    Position Offset Coordinate Space=::Local
    Sprite UV Mode=::NewEnumerator3
    UsePositionOffset=true
    Write Material Random=true
   ) {module=InitializeParticle001}
   &SkeletalMeshLocation(
    Bone / Socket Index=2
    MaskSpawnGroup=false
    Mesh Sampling Type=::NewEnumerator5
    Skeletal Mesh=$User.SkeletalMesh
    Socket Sampling Mode=::NewEnumerator5
   )
   &ShapeLocation(
    MaskSpawnGroup=false
    Offset=<0 0 10>
    Offset Mode=::NewEnumerator0
    Override Seed=false
    Shape Origin=$Particles.SkeletalMeshLocation.SampledPosition
    Sphere Radius=15
    UseOwnerScale=true
   )
   &MeshRotationForce(Rotation=&VectorFromFloat(Value=1))
   &InitialMeshOrientation(Use Rotation Vector=false) {module=InitialMeshOrientation001}
   &AddVelocity(Velocity=&RandomRangeVector(Minimum=<-1 -1 -1>, Maximum=<1 1 1>), Velocity Speed Scale=42.5983) {enabled=false, module=AddVelocity002}
   &InheritVelocity(
    Inherited Velocity Amount Scale=&Multiply_VectorByFloat(
     Vector=<1 1 1>
     Float=&Multiply_Float(A=$System.InheritVelocityElements, B=&RandomRangeFloat(Minimum=0.5, Maximum=1))
    )
    Inherited Velocity Speed Limit=35
    Speed Limit=true
    UseSpeedThreshold=false
   )
   &ApplyInitialForces()
   &CameraOffset(Camera Offset Amount=&RandomRangeFloat(Minimum=0, Maximum=25))
   /SetVariables(Particles.Color=&MultiplyLinearColor(
     Linear Color=$Particles.Color
     Scale Factor=&MakeFloat4FromVectorAndFloat("Vector (RGB)"=&VectorFromFloat(Value=$System.TOD), "Float (Alpha)"=1)
    )) {enabled=false, isSetParameters=true}
   /SetVariables(Particles.MeshIndex=&MakeCustomIntFromBool(
     Bool=$System.VarientBase
     True INT=&RandomRangeInt(Minimum=0, Maximum=3)
     False INT=&RandomRangeInt(Minimum=1, Maximum=3)
    )) {isSetParameters=true}
   &DynamicMaterialParameters(
    Index 0 Param 1=&MakeFloatFromInt(Int=$Particles.MeshIndex)
    Module.Index0Param0Write=true
    Module.Index0Param1Write=false
    Module.Index0Param2Write=false
    Module.Index0Param3Write=false
   )
  ]
  particleUpdateScript [
   &ParticleState()
   &Drag(Drag=&RandomRangeFloat(Minimum=6, Maximum=12), Use Linear Drag=true, Use Rotational Drag=false) {module=Drag001}
   &GravityForce(Gravity=&RandomRangeVector(Minimum=<0 0 150>, Maximum=<0 0 -150>))
   &CurlNoiseForce(Pan Noise=false)
   &ScaleMeshSize(Scale Factor=&Multiply_VectorByFloat(
     Vector=<1 1 1>
     Float=&FloatFromCurve(FloatCurve=DataInterface() {propertyValues {Curve {keys [
          {interpMode=RCIM_Cubic, tangentMode=RCTM_User, time=0, value=0}
          <at 0.2 1>
          {interpMode=RCIM_Cubic, tangentMode=RCTM_User, time=0.5, value=1}
          {arriveTangent=-0.9819, interpMode=RCIM_Cubic, leaveTangent=-0.9819, tangentMode=RCTM_User, time=0.75, value=1.3}
          {arriveTangent=-9.5654, interpMode=RCIM_Cubic, leaveTangent=-9.5654, tangentMode=RCTM_User, time=1, value=0}
         ]}}})
    )) {module=ScaleMeshSize001}
   &SolveForcesAndVelocity(Clamp Velocity=false, Limit Acceleration=false)
   /SetVariables(
    Particles.Position=&AddVectorToPosition(
     Position=$Particles.Position
     Vector=&Multiply_Vector(
      A=&VectorFromFloat(Value=0.6)
      B=&SelectVectorFromArray(
       Direct Array Index=&ModuloInt(
        A=&MakeIntFromFloat(Float=&Add_Float(
          A=&Multiply_Float(A=$Engine.Time, B=6)
          B=&MakeFloatFromInt(Int=$Particles.SpawnGroup)
         ))
        B=6
       )
       Vector Selection Array=DataInterface() {propertyValues {FloatData [<1 0 0>, <0 -0.5 -0.5>, <-1 0.5 0.5>, <0.5 2 1>, <0.1 -0.4 0.2>, <0.35 0.7 -0.4>]}}
      )
     )
    )
    Particles.Scale=&Multiply_Vector(
     A=$Particles.Scale
     B=&SelectVectorFromArray(
      Direct Array Index=&ModuloInt(
       A=&MakeIntFromFloat(Float=&Add_Float(
         A=&Multiply_Float(A=$Engine.Time, B=6)
         B=&MakeFloatFromInt(Int=$Particles.SpawnGroup)
        ))
       B=3
      )
      Vector Selection Array=DataInterface() {propertyValues {FloatData [<1 1 1>, <1.05 1 1.05>, <0.95 1.02 0.99>]}}
     )
    )
   ) {enabled=false, isSetParameters=true}
   &GenerateDeathEvent(Use Event Probability=false) {enabled=false}
  ]
  rendererClasses [&NiagaraMeshRendererProperties]
  renderers [
   {rendererClass=&NiagaraMeshRendererProperties, rendererIndex=0}
   {rendererClass=&NiagaraMeshRendererProperties, rendererIndex=1}
  ]
 }
 Round_Trail {
  emitterUpdateScript [
   &EmitterState(Life Cycle Mode=::NewEnumerator1, Loop Behavior=::NewEnumerator1, Loop Duration=0.5, UseLoopDelay=false)
   &FN_SpawnAnimTrail()
   /SetVariables(Emitter.Random:NiagaraFloat=&RandomRangeFloat(Minimum=0, Maximum=0.1)) {isSetParameters=true, module=SetVariables_#3}
  ]
  particleSpawnScript [
   &FN_InitializeAnimTrailBoneIndecies()
   &FN_InitializeAnimTrail(Color=<rgba 1 1 1 0.5>, Lifetime=0.2)
   &DynamicMaterialParameters(
    Index 0 Param 4=$Emitter.Random
    Module.Param0WriteEnabled=false
    Module.Param1WriteEnabled=false
    Module.Param2WriteEnabled=false
    Module.Param3WriteEnabled=true
   ) {module=DynamicMaterialParameters001}
  ]
  particleUpdateScript [
   &FN_UpdateAnimTrail(SkeletalMesh=$User.SkeletalMesh) {module=FN_UpdateAnimTrail001}
   &SolveForcesAndVelocity(Clamp Velocity=false, Limit Acceleration=false)
   &ScaleColor(
    Scale Alpha=&Multiply_Float(
     A=&FloatFromCurve(FloatCurve=DataInterface() {propertyValues {Curve {keys [
          {arriveTangent=-1.8862, interpMode=RCIM_Cubic, leaveTangent=-1.8862, tangentMode=RCTM_User, time=0.2, value=1.25}
          {interpMode=RCIM_Cubic, tangentMode=RCTM_User, time=1, value=0}
         ]}}})
     B=&Lerp_Float(
      A=1
      B=&FloatFromCurve(FloatCurve=DataInterface() {propertyValues {Curve {keys [
           {arriveTangent=0.6814, interpMode=RCIM_Cubic, leaveTangent=0.6814, tangentMode=RCTM_User, time=0, value=0.1}
           {arriveTangent=2.5, interpMode=RCIM_Cubic, leaveTangent=2.5, time=0.15, value=0.3}
           {arriveTangent=1.5958, time=0.2, value=0.6}
           {arriveTangent=0.0224, interpMode=RCIM_Cubic, leaveTangent=0.0224, tangentMode=RCTM_User, time=1, value=1}
          ]}}})
      Alpha=&FloatFromCurve(
       FloatCurve=DataInterface() {propertyValues {Curve {keys [<at 0 0 RCIM_Constant>,<at 0.4 1 RCIM_Constant>]}}}
       CurveIndex=$Emitter.NormalizedLoopAge
      )
     )
    )
    Scale RGB=&Multiply_VectorByFloat(Vector=<1 1 1>, Float=&FNTimeOfDayScalar(Night Scale=0.33, Morning Scale=1, Evening Scale=1))
    ScaleA=true
    ScaleRGB=true
   ) {module=ScaleColor001}
   &ScaleColor(
    Scale Alpha=&FloatFromCurve(FloatCurve=DataInterface() {propertyValues {Curve {keys [
         {arriveTangent=0.6576, interpMode=RCIM_Cubic, leaveTangent=0.6576, tangentMode=RCTM_User, time=0, value=0}
         {arriveTangent=1.9656, interpMode=RCIM_Cubic, leaveTangent=1.9656, tangentMode=RCTM_User, time=0.15, value=0.6228}
         {arriveTangent=0.0224, interpMode=RCIM_Cubic, leaveTangent=0.0224, tangentMode=RCTM_User, time=1, value=1}
        ]}}})
    ScaleA=true
    ScaleRGB=true
   ) {enabled=false}
   &ScaleRibbonWidth(Ribbon Width Scale=&FloatFromCurve(
     FloatCurve=DataInterface() {propertyValues {Curve {keys [{arriveTangent=-1.7104, interpMode=RCIM_Cubic, leaveTangent=-1.7104, tangentMode=RCTM_User, time=0.4, value=1}, <at 1 0.5 RCIM_Cubic>]}}}
     CurveIndex=$Emitter.NormalizedLoopAge
    ))
   &DynamicMaterialParameters(
    Index 0 Param 3=&FloatFromCurve(
     FloatCurve=DataInterface() {propertyValues {Curve {keys [<at 0.25 0 RCIM_Cubic>,<at 0.4 1 RCIM_Cubic>]}}}
     CurveIndex=$Emitter.NormalizedLoopAge
    )
    Module.Param0WriteEnabled=false
    Module.Param1WriteEnabled=false
    Module.Param2WriteEnabled=true
    Module.Param3WriteEnabled=false
   )
   /SetVariables(Particles.Color=&MultiplyLinearColor(
     Linear Color=$Particles.Color
     Scale Factor=&MakeVector4(X=1, Y=1, Z=1, W=&MakeCustomFloatFromBool(Bool=$System.VarientBase, True Float=1, False Float=2))
    )) {isSetParameters=true}
  ]
  rendererClasses [&NiagaraRibbonRendererProperties]
  renderers [
   {rendererClass=&NiagaraRibbonRendererProperties}
   {rendererClass=&NiagaraRibbonRendererProperties, rendererIndex=1}
  ]
 }
}
summary {
 systemName=NS_Pickaxe_IndieBucket_Trail
 emitters [
  {emitterName=Round_Trail, rendererClasses [&NiagaraRibbonRendererProperties]}
  {emitterName=Outline, rendererClasses [&NiagaraRibbonRendererProperties]}
  {emitterName=Pop, rendererClasses [&NiagaraMeshRendererProperties]}
 ]
 userVariables [
  User.ColorSwitcher {type=NiagaraFloat, defaultValue=0, description=""}
  User.MeshColors {
   type=NiagaraDataInterfaceArrayColor
   description=""
   defaultValue {
    struct=NiagaraExt_VariableValue_DataInterface
    value {
     dataInterface=&self:NiagaraDataInterfaceArrayColor_0
     dataInterfaceClass=&NiagaraDataInterfaceArrayColor
    }
   }
  }
  User.MeshColors_Variant_01 {
   type=NiagaraDataInterfaceArrayColor
   description=""
   defaultValue {
    struct=NiagaraExt_VariableValue_DataInterface
    value {
     dataInterface=&self:NiagaraDataInterfaceArrayColor_1
     dataInterfaceClass=&NiagaraDataInterfaceArrayColor
    }
   }
  }
  User.RibbonIDOffset {type=NiagaraInt32}
  User.SkeletalMesh {
   type=NiagaraDataInterfaceSkeletalMesh
   defaultValue {
    struct=NiagaraExt_VariableValue_DataInterface
    value {dataInterface=&self:NiagaraDataInterfaceSkeletalMesh_2}
   }
  }
  User.VariantSwitch {type=NiagaraFloat, defaultValue=0, description=""}
 ]
}
