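# Physics asset for the IndieBucket two-handed pickaxe, named as the frontend variant, defining physics for the weapon's handle and two dynamic ear bones.
# Each of the three bodies is a single capsule: a long, thin one on the handle and a shorter, wider one on each ear.
# Both ears constrain directly to the handle with tight twist limits and asymmetric swing ranges that differ slightly between left and right.
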PhysicsAsset /BRCosmetics/Weapons/FORT_Melee/Pickaxe_IndieBucket/Meshes/IndieBucket_Axe_Physics_FRONTEND
bodies {
 Melee_TwoHanded_Handle [{shapeType=Capsule, length=47.4488, radius=3.0722, center=<0.4979 -1.4834 9.8163>, rotation=<rot 0.0144 -90 -0.0002>}]
 dyn_ear_l_0 [{shapeType=Capsule, length=8.3, radius=5.7, shapeName=dyn_ear_l_0_capsule, center=<8 0 1>, rotation=<rot 85 -180 180>}]
 dyn_ear_r_0 [{shapeType=Capsule, length=8.3, radius=5.7, shapeName=dyn_ear_r_0_capsule, center=<8 0 1>, rotation=<rot 85 180 180>}]
}
constraints [
 {bone1Name=dyn_ear_l_0, bone2Name=Melee_TwoHanded_Handle, swing1LimitDegrees=10, swing2LimitDegrees=50, twistLimitDegrees=5}
 {bone1Name=dyn_ear_r_0, bone2Name=Melee_TwoHanded_Handle, swing1LimitDegrees=15, swing2LimitDegrees=45, twistLimitDegrees=5}
]
