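# Anim Blueprint for the IndieBucket pickaxe that adds secondary physics motion to its ear and wire bones.
# In game, the input pose runs through chained AnimDynamics nodes with partial alphas on most bones.
# When bIsInFrontEnd is true, it switches to a RigidBody simulation using a front-end physics asset on an unfed cached pose.
# A second, full-alpha ear chain is unconnected to the output, and the EventGraph does nothing.
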AnimBlueprint /BRCosmetics/Weapons/FORT_Melee/Pickaxe_IndieBucket/Meshes/IndieBucket_Axe_AnimBP
target_skeleton=&/IndieBucket_Axe_Skeleton
graphs [
 AnimGraph {nodes [
   n1 = Anim Dynamics - Bone: dyn_ear_r_0(Alpha=1, ExternalForce="0,0,0") {
    angularDampingOverride=0.5
    angularSpringConstant=25
    bOverrideAngularDamping=true
    bUseGravityOverride=true
    simulationSpace=BoneRelative
    alphaBoolBlend {blendOption=Linear}
    boundBone {boneName=dyn_ear_r_0}
    componentAppliedLinearAccClamp=<3000 3000 3000>
    componentLinearAccScale=<0.1 0.5 0.3>
    componentLinearVelScale=<0.1 0.5 0.3>
    gravityOverride {z=-980}
    physicsBodyDefinitions [{
      boundBone {boneName=dyn_ear_r_0}
      boxExtents=<15 7.5 2>
      constraintSetup {angularLimitsMax=<0.1 70 40>, angularLimitsMin=<-0.1 -90 -40>, angularTarget=<0 0 0>}
      localJointOffset=<-5 0.5 0>
     }]
    relativeSpaceBone {boneName=b_AR_Root}
   }
   n2 = Anim Dynamics - Bone: dyn_ear_r_1(ComponentPose=@n1, Alpha=1, ExternalForce="0,0,0") {
    angularDampingOverride=0.5
    angularSpringConstant=5000
    bOverrideAngularDamping=true
    bUseGravityOverride=true
    simulationSpace=BoneRelative
    alphaBoolBlend {blendOption=Linear}
    boundBone {boneName=dyn_ear_r_1}
    componentAppliedLinearAccClamp=<3000 3000 3000>
    componentLinearAccScale=<0.2 0.5 0.2>
    componentLinearVelScale=<0.1 0.25 0.1>
    gravityOverride {z=-600}
    physicsBodyDefinitions [{
      boundBone {boneName=dyn_ear_r_1}
      boxExtents=<10 20 20>
      constraintSetup {angularLimitsMax=<0.1 90 2.5>, angularLimitsMin=<-0.1 -90 -2.5>, angularTarget=<0 0 0>}
      localJointOffset=<-3 0 0.15>
     }]
    relativeSpaceBone {boneName=b_AR_Root}
   }
   n3 = Anim Dynamics - Bone: dyn_ear_l_0(ComponentPose=@n2, Alpha=1, ExternalForce="0,0,0") {
    angularDampingOverride=0.5
    angularSpringConstant=25
    bOverrideAngularDamping=true
    bUseGravityOverride=true
    simulationSpace=BoneRelative
    alphaBoolBlend {blendOption=Linear}
    boundBone {boneName=dyn_ear_l_0}
    componentAppliedLinearAccClamp=<3000 3000 3000>
    componentLinearAccScale=<0.1 0.5 0.3>
    componentLinearVelScale=<0.1 0.5 0.3>
    gravityOverride {z=-980}
    physicsBodyDefinitions [{
      boundBone {boneName=dyn_ear_l_0}
      boxExtents=<15 7.5 2>
      constraintSetup {angularLimitsMax=<0.1 70 40>, angularLimitsMin=<-0.1 -90 -40>, angularTarget=<0 0 0>}
      localJointOffset=<-5 0.5 0>
     }]
    relativeSpaceBone {boneName=b_AR_Root}
   }
   Anim Dynamics - Bone: dyn_ear_l_1(ComponentPose=@n3, Alpha=1, ExternalForce="0,0,0") {
    angularDampingOverride=0.5
    angularSpringConstant=5000
    bOverrideAngularDamping=true
    bUseGravityOverride=true
    simulationSpace=BoneRelative
    alphaBoolBlend {blendOption=Linear}
    boundBone {boneName=dyn_ear_l_1}
    componentAppliedLinearAccClamp=<3000 3000 3000>
    componentLinearAccScale=<0.2 0.5 0.2>
    componentLinearVelScale=<0.1 0.25 0.1>
    gravityOverride {z=-600}
    physicsBodyDefinitions [{
      boundBone {boneName=dyn_ear_l_1}
      boxExtents=<10 20 20>
      constraintSetup {angularLimitsMax=<0.1 90 2.5>, angularLimitsMin=<-0.1 -90 -2.5>, angularTarget=<0 0 0>}
      localJointOffset=<-3 0 0.15>
     }]
    relativeSpaceBone {boneName=b_AR_Root}
   }
   n4 = Anim Dynamics - Bone: dyn_ear_r_0(ComponentPose=Local To Component(LocalPose=Use cached pose 'In-game'()), Alpha=0.45, ExternalForce="0,0,0") {
    angularDampingOverride=0.99
    angularSpringConstant=25
    bOverrideAngularDamping=true
    bUseGravityOverride=true
    simulationSpace=BoneRelative
    alphaBoolBlend {blendOption=Linear}
    boundBone {boneName=dyn_ear_r_0}
    componentAppliedLinearAccClamp=<3000 3000 3000>
    componentLinearAccScale=<0.1 0.5 0.3>
    componentLinearVelScale=<0.1 0.5 0.3>
    gravityOverride {z=-980}
    physicsBodyDefinitions [{
      boundBone {boneName=dyn_ear_r_0}
      boxExtents=<15 7.5 2>
      constraintSetup {angularLimitsMax=<0.1 70 40>, angularLimitsMin=<-0.1 -90 -40>, angularTarget=<0 0 0>}
      localJointOffset=<-5 0.5 0>
     }]
    relativeSpaceBone {boneName=b_AR_Root}
   }
   n5 = Anim Dynamics - Bone: dyn_ear_r_1(ComponentPose=@n4, Alpha=0.1, ExternalForce="0,0,0") {
    angularDampingOverride=0.99
    angularSpringConstant=5000
    bOverrideAngularDamping=true
    bUseGravityOverride=true
    simulationSpace=BoneRelative
    alphaBoolBlend {blendOption=Linear}
    boundBone {boneName=dyn_ear_r_1}
    componentAppliedLinearAccClamp=<3000 3000 3000>
    componentLinearAccScale=<0.2 0.5 0.2>
    componentLinearVelScale=<0.1 0.25 0.1>
    gravityOverride {z=-600}
    physicsBodyDefinitions [{
      boundBone {boneName=dyn_ear_r_1}
      boxExtents=<10 20 20>
      constraintSetup {angularLimitsMax=<0.1 90 2.5>, angularLimitsMin=<-0.1 -90 -2.5>, angularTarget=<0 0 0>}
      localJointOffset=<-3 0 0.15>
     }]
    relativeSpaceBone {boneName=b_AR_Root}
   }
   n6 = Anim Dynamics - Bone: dyn_ear_l_0(ComponentPose=@n5, Alpha=0.5, ExternalForce="0,0,0") {
    angularDampingOverride=0.9
    angularSpringConstant=25
    bOverrideAngularDamping=true
    bUseGravityOverride=true
    simulationSpace=BoneRelative
    alphaBoolBlend {blendOption=Linear}
    boundBone {boneName=dyn_ear_l_0}
    componentAppliedLinearAccClamp=<3000 3000 3000>
    componentLinearAccScale=<0.1 0.5 0.3>
    componentLinearVelScale=<0.1 0.5 0.3>
    gravityOverride {z=-980}
    physicsBodyDefinitions [{
      boundBone {boneName=dyn_ear_l_0}
      boxExtents=<15 7.5 2>
      constraintSetup {angularLimitsMax=<0.1 70 40>, angularLimitsMin=<-0.1 -90 -40>, angularTarget=<0 0 0>}
      localJointOffset=<-5 0.5 0>
     }]
    relativeSpaceBone {boneName=b_AR_Root}
   }
   n7 = Anim Dynamics - Bone: dyn_ear_l_1(ComponentPose=@n6, Alpha=0.1, ExternalForce="0,0,0") {
    angularDampingOverride=0.95
    angularSpringConstant=5000
    bOverrideAngularDamping=true
    bUseGravityOverride=true
    simulationSpace=BoneRelative
    alphaBoolBlend {blendOption=Linear}
    boundBone {boneName=dyn_ear_l_1}
    componentAppliedLinearAccClamp=<3000 3000 3000>
    componentLinearAccScale=<0.2 0.5 0.2>
    componentLinearVelScale=<0.1 0.25 0.1>
    gravityOverride {z=-600}
    physicsBodyDefinitions [{
      boundBone {boneName=dyn_ear_l_1}
      boxExtents=<10 20 20>
      constraintSetup {angularLimitsMax=<0.1 90 2.5>, angularLimitsMin=<-0.1 -90 -2.5>, angularTarget=<0 0 0>}
      localJointOffset=<-3 0 0.15>
     }]
    relativeSpaceBone {boneName=b_AR_Root}
   }
   n8 = Anim Dynamics - Bone: dyn_wireBlue(ComponentPose=@n7, Alpha=1, ExternalForce="0,0,0") {
    angularDampingOverride=0.999
    bOverrideAngularDamping=true
    bUseGravityOverride=true
    linearDampingOverride=0.99
    linearSpringConstant=310
    simulationSpace=RootRelative
    alphaBoolBlend {blendOption=Linear}
    boundBone {boneName=dyn_wireBlue}
    componentAppliedLinearAccClamp=<3000 3000 3000>
    componentLinearAccScale=<0.5 0.2 0.2>
    componentLinearVelScale=<0.25 0.1 0.1>
    gravityOverride {z=-980}
    physicsBodyDefinitions [{
      boundBone {boneName=dyn_wireBlue}
      boxExtents=<10 10 5>
      constraintSetup {angularLimitsMax=<0.2 0 0.2>, angularLimitsMin=<-0.2 -30 -0.2>, angularTarget=<0 0 0>}
      localJointOffset=<-3 0 0>
     }]
   }
   n9 = Anim Dynamics - Bone: dyn_wirePink(ComponentPose=@n8, Alpha=0.4, ExternalForce="0,0,0") {
    angularDampingOverride=0.9999
    bOverrideAngularDamping=true
    bUseGravityOverride=true
    linearDampingOverride=0.99
    linearSpringConstant=310
    simulationSpace=RootRelative
    alphaBoolBlend {blendOption=Linear}
    boundBone {boneName=dyn_wirePink}
    componentAppliedLinearAccClamp=<3000 3000 3000>
    componentLinearAccScale=<0.3 0.5 0.2>
    componentLinearVelScale=<0.15 0.15 0.1>
    gravityOverride {z=-980}
    physicsBodyDefinitions [{
      boundBone {boneName=dyn_wirePink}
      boxExtents=<5 10 10>
      constraintSetup {angularLimitsMax=<0.2 0.2 0>, angularLimitsMin=<-0.2 -0.2 -45>, angularTarget=<0 0 0>}
      localJointOffset=<-3 0 0>
     }]
   }
   n10 = RigidBody(ComponentPose=Local To Component(LocalPose=Use cached pose 'frontEnd'())) {
    bOverrideWorldGravity=true
    overridePhysicsAsset=&/IndieBucket_Axe_Physics_FRONTEND
    simulationSpace=BaseBoneSpace
    alphaBoolBlend {blendOption=Linear}
    baseBoneRef {boneName=b_AR_Root}
    componentLinearAccScale=<0.2 0.5 0.2>
    componentLinearVelScale=<0.125 0.25 0.125>
    overrideWorldGravity {z=-980}
   }
   Output Pose(Result=Blend Poses by bool(BlendPose_0=Component To Local(ComponentPose=@n10), BlendPose_1=Component To Local(ComponentPose=@n9), bActiveValue=Get bIsInFrontEnd()) {blendType=Linear}) {layerGroup=None}
   AnimGraphNode_SubInput_0 = Input Pose()
   Save cached pose 'In-game'(Pose=@AnimGraphNode_SubInput_0) {cacheName=In-game}
   Save cached pose 'frontEnd'() {cacheName=frontEnd}
  ]}
 EventGraph {nodes [
   TryGetPawnOwner()
   Event BlueprintUpdateAnimation()
  ]}
]
variables []
